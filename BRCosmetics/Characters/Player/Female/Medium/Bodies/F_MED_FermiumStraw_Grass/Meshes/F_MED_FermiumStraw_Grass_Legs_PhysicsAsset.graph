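# Physics asset for the legs mesh of the FermiumStraw_Grass medium female player cosmetic body.
# Capsule bodies cover the pelvis, both feet and the left and right kin_pouches bones, with mirrored placement between sides.
# Thigh and calf bodies are listed without any shapes.
# No constraints link the bodies, so the asset defines no jointed ragdoll chain.
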PhysicsAsset /BRCosmetics/Characters/Player/Female/Medium/Bodies/F_MED_FermiumStraw_Grass/Meshes/F_MED_FermiumStraw_Grass_Legs_PhysicsAsset
bodies {
 calf_l []
 calf_r []
 foot_l [{shapeType=Capsule, length=19.3444, radius=8.0344, center=<-4.1811 6.0027 -0.338>, rotation=<rot 0.0388 1.4701 -89.3982>}]
 foot_r [{shapeType=Capsule, length=19.3444, radius=8.0344, center=<4.1811 -6.0027 0.338>, rotation=<rot 0.0388 1.4701 90.6018>}]
 kin_pouches_l [{shapeType=Capsule, length=8.6112, radius=8.4139, shapeName=kin_pouches_l_capsule, center=<5.3331 0.0524 0.3333>, rotation=<rot 90 180 0>}]
 kin_pouches_r [{shapeType=Capsule, length=8.6112, radius=8.4139, shapeName=kin_pouches_r_capsule, center=<5.3331 -0.0524 -0.3333>, rotation=<rot 90 0 0>}]
 pelvis [{shapeType=Capsule, length=9.9744, radius=11.6498, center=<-4.1495 -1.5798 -0.1981>}]
 thigh_l []
 thigh_r []
}
constraints []
