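# Post-process style cosmetic AnimBP for the female medium FermiumStraw Grass torso, adding secondary motion to armbands and chest spray cans.
# It copies the attached parent's pose, runs a torso Control Rig, and caches the result.
# Four chained root-relative Anim Dynamics nodes with angular springs follow, with spray cans at reduced alpha and no gravity.
# On customization load, a sim-space settings variable is set differently for front end and other contexts.
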AnimBlueprint /BRCosmetics/Characters/Player/Female/Medium/Bodies/F_MED_FermiumStraw_Grass/Meshes/F_MED_FermiumStraw_Grass_Torso_AnimBP
target_skeleton=&/F_MED_FermiumStraw_Grass_Skeleton
graphs [
 AnimGraph {nodes [
   n1 = Anim Dynamics - Bone: dyn_armband_2!(
    ComponentPose=Local To Component(LocalPose=Use cached pose 'controlRigCP'())
    Alpha=1
    SimSpaceSettings:Anim Phys Sim Space Settings Structure=(SimSpaceAngularAlpha=0,MaxAngularVelocity=10000,MaxAngularAcceleration=10000,ExternalAngularVelocity=(X=0,Y=0,Z=0))
   ) -> (Pose) {
    angularDampingOverride=1
    angularSpringConstant=10000
    bAngularSpring=true
    bOverrideAngularDamping=true
    bUseGravityOverride=true
    gravityScale=0.35
    simulationSpace=RootRelative
    boundBone {boneName=dyn_armband_2}
    componentAppliedLinearAccClamp=<1500 1500 1500>
    componentLinearAccScale=<1 1 1>
    componentLinearVelScale=<1 1 1>
    gravityOverride {z=-980}
    physicsBodyDefinitions [{
      boundBone {boneName=dyn_armband_2}
      boxExtents=<20 30 30>
      constraintSetup {angularLimitsMax=<0 2 15>, angularLimitsMin=<0 -15 -2>, angularTarget=<1 0 0>}
      localJointOffset=<8 0 0>
     }]
   }
   n2 = Anim Dynamics - Bone: dyn_armband_1!(
    ComponentPose=@n1
    Alpha=1
    SimSpaceSettings:Anim Phys Sim Space Settings Structure=(SimSpaceAngularAlpha=0,MaxAngularVelocity=10000,MaxAngularAcceleration=10000,ExternalAngularVelocity=(X=0,Y=0,Z=0))
   ) -> (Pose) {
    angularDampingOverride=1
    angularSpringConstant=10000
    bAngularSpring=true
    bOverrideAngularDamping=true
    bUseGravityOverride=true
    gravityScale=0.35
    simulationSpace=RootRelative
    boundBone {boneName=dyn_armband_1}
    componentAppliedLinearAccClamp=<1500 1500 1500>
    componentLinearAccScale=<1 1 1>
    componentLinearVelScale=<1 1 1>
    gravityOverride {z=-980}
    physicsBodyDefinitions [{
      boundBone {boneName=dyn_armband_1}
      boxExtents=<20 30 30>
      constraintSetup {angularLimitsMax=<0 15 15>, angularLimitsMin=<0 -15 -15>, angularTarget=<1 0 0>}
      localJointOffset=<8 0 0>
     }]
   }
   n3 = Anim Dynamics - Bone: dyn_sprayCanChest_1!(
    ComponentPose=@n2
    Alpha=0.375
    SimSpaceSettings:Anim Phys Sim Space Settings Structure=(SimSpaceAngularAlpha=0,MaxAngularVelocity=10000,MaxAngularAcceleration=10000,ExternalAngularVelocity=(X=0,Y=0,Z=0))
   ) -> (Pose) {
    angularDampingOverride=1
    angularSpringConstant=15000
    bAngularSpring=true
    bOverrideAngularDamping=true
    gravityScale=0
    simulationSpace=RootRelative
    alphaBoolBlend {blendOption=Linear}
    boundBone {boneName=dyn_sprayCanChest_1}
    componentAppliedLinearAccClamp=<3000 3000 3000>
    componentLinearAccScale=<0.4 0.4 0.4>
    componentLinearVelScale=<0.4 0.4 0.4>
    gravityOverride {z=-980}
    physicsBodyDefinitions [{
      boundBone {boneName=dyn_sprayCanChest_1}
      boxExtents=<20 20 20>
      constraintSetup {angularLimitsMax=<0 10 5>, angularLimitsMin=<0 -10 -5>, angularTarget=<1 0 0>}
      localJointOffset=<2 0 0>
     }]
   }
   n4 = Anim Dynamics - Bone: dyn_sprayCanChest_2!(
    ComponentPose=@n3
    Alpha=0.375
    SimSpaceSettings:Anim Phys Sim Space Settings Structure=(SimSpaceAngularAlpha=0,MaxAngularVelocity=10000,MaxAngularAcceleration=10000,ExternalAngularVelocity=(X=0,Y=0,Z=0))
   ) -> (Pose) {
    angularDampingOverride=1
    angularSpringConstant=15000
    bAngularSpring=true
    bOverrideAngularDamping=true
    gravityScale=0
    simulationSpace=RootRelative
    alphaBoolBlend {blendOption=Linear}
    boundBone {boneName=dyn_sprayCanChest_2}
    componentAppliedLinearAccClamp=<3000 3000 3000>
    componentLinearAccScale=<0.4 0.4 0.4>
    componentLinearVelScale=<0.4 0.4 0.4>
    gravityOverride {z=-980}
    physicsBodyDefinitions [{
      boundBone {boneName=dyn_sprayCanChest_2}
      boxExtents=<20 20 20>
      constraintSetup {angularLimitsMax=<0 10 5>, angularLimitsMin=<0 -10 -5>, angularTarget=<1 0 0>}
      localJointOffset=<2 0 0>
     }]
   }
   Output Pose(Result=Component To Local(ComponentPose=@n4)) {layerGroup=None}
   n5 = Control Rig(Source=Copy Pose From Mesh() {bUseAttachedParent=true}) {controlRigAssetReference {controlRigAsset=&/F_MED_FermiumStraw_Grass_Torso_CtrlRig}}
   Save cached pose 'controlRigCP'(Pose=@n5) {cacheName=controlRigCP}
  ]}
 EventGraph {nodes [
   Event OnCharacterCustomizationLoaded()
   > n1 = Branch(Condition=Get bIsInFrontEnd())
   > Set!(
    execute
    Sim Space Settings output_SimSpaceAngularAlpha:float=1
    Sim Space Settings output_MaxAngularVelocity:float=10
    Sim Space Settings output_MaxAngularAcceleration:float=1
    Sim Space Settings output_ExternalAngularVelocity:Vector="0,0,0"
   ) -> (then, Output_Get:Anim Phys Sim Space Settings Structure)
   Set(
    execute=@n1.else
    Sim Space Settings output_SimSpaceAngularAlpha=0
    Sim Space Settings output_MaxAngularVelocity=10000
    Sim Space Settings output_MaxAngularAcceleration=10000
    Sim Space Settings output_ExternalAngularVelocity="0,0,0"
   )
  ]}
]
variables [Sim Space Settings output {
  type=AnimPhysSimSpaceSettings
  default {maxAngularAcceleration=10000, maxAngularVelocity=10000, simSpaceAngularAlpha=0, externalAngularVelocity=<0 0 0>}
 }]
